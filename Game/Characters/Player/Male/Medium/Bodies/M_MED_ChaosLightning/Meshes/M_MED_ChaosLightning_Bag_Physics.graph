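# Physics asset for the bag on the M_MED_ChaosLightning player body, setting up purse and strap dynamics with capsule bodies.
# A three-link purse chain and a two-link strap chain hang from base bodies locked in rotation to spine_03, with tight swing and twist limits.
# A further strap link branches off the second purse link through its own base body.
# Larger capsules approximate the spine and right thigh, and the thigh has no constraint.
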PhysicsAsset /Game/Characters/Player/Male/Medium/Bodies/M_MED_ChaosLightning/Meshes/M_MED_ChaosLightning_Bag_Physics
bodies {
 base_purseStrap [{shapeType=Capsule, length=0.505, radius=0.505, shapeName=base_purseStrap_capsule, rotation=<rot 0 -90 0>}]
 base_purseStrap03 [{shapeType=Capsule, length=0.505, radius=0.505, shapeName=base_purseStrap03_capsule, rotation=<rot 0 -90 0>}]
 base_purse_01 [{shapeType=Capsule, length=0.505, radius=0.505, shapeName=base_purse_01_capsule, rotation=<rot 0 -90 0>}]
 dyn_purseStrap01 [{shapeType=Capsule, length=2, radius=1.2, shapeName=dyn_purseStrap01_capsule, center=<5.2909 0 0>, rotation=<rot 0 0 90>}]
 dyn_purseStrap02 [{shapeType=Capsule, length=11.5, radius=2.8, shapeName=dyn_purseStrap02_capsule, center=<8.7336 0 0.4313>, rotation=<rot 85.1258 0 0>}]
 dyn_purseStrap03 [{
   shapeType=Capsule
   length=7.7
   radius=2.1
   shapeName=dyn_purseStrap03_capsule
   center=<5.5691 -2.6202 -0.0664>
   rotation=<rot 80.2117 -179.9996 -108.3057>
  }]
 dyn_purse_01 [{shapeType=Capsule, length=3, radius=3, shapeName=dyn_purse_01_capsule, center=<4 0 0>, rotation=<rot 90 0 0>}]
 dyn_purse_02 [{shapeType=Capsule, length=3, radius=3, shapeName=dyn_purse_02_capsule, center=<4 0 0>, rotation=<rot 90 0 0>}]
 dyn_purse_03 [
  {
   shapeType=Capsule
   length=12.1
   radius=4.4
   shapeName=dyn_purse_03_capsule
   center=<11.4888 -7.9121 -1.2533>
   rotation=<rot -25.0404 -75.9912 62.2087>
  }
  {
   shapeType=Capsule
   length=12.1
   radius=4.4
   shapeName=dyn_purse_03_capsule
   center=<11.0121 7.7464 -0.3695>
   rotation=<rot -24.0234 -73.5976 61.9016>
  }
 ]
 spine_01 [{shapeType=Capsule, length=8.88, radius=12.9114, shapeName=spine_01_capsule, center=<-6.6707 -0.3373 -0.7515>}]
 spine_03 [{
   shapeType=Capsule
   length=9.6693
   radius=12.3928
   shapeName=spine_03_capsule
   center=<-3.9229 -0.4322 -3.5932>
   rotation=<rot 90 180 178.122>
  }]
 thigh_r [{
   shapeType=Capsule
   length=18.8059
   radius=9.364
   shapeName=thigh_l_capsule
   center=<11.8171 -2.5048 -0.3619>
   rotation=<rot 81.0292 178.3909 -10.9513>
  }]
}
constraints [
 {bone1Name=dyn_purse_03, bone2Name=dyn_purse_02, swing1LimitDegrees=10, swing2LimitDegrees=25, twistLimitDegrees=5}
 {bone1Name=dyn_purseStrap03, bone2Name=base_purseStrap03, swing1LimitDegrees=30, swing2LimitDegrees=40, twistLimitDegrees=1}
 {bone1Name=dyn_purse_02, bone2Name=dyn_purse_01, swing1LimitDegrees=10, swing2LimitDegrees=10, twistLimitDegrees=5}
 {bone1Name=dyn_purse_01, bone2Name=base_purse_01, swing1LimitDegrees=8, swing2LimitDegrees=10, twistLimitDegrees=5}
 {bone1Name=dyn_purseStrap02, bone2Name=dyn_purseStrap01, swing1LimitDegrees=5, swing2LimitDegrees=40, twistLimitDegrees=1}
 {bone1Name=dyn_purseStrap01, bone2Name=base_purseStrap, swing1LimitDegrees=2, swing2LimitDegrees=8, twistLimitDegrees=1}
 {
  bone1Name=base_purse_01
  bone2Name=spine_03
  swing1LimitDegrees=45
  swing1Motion=Locked
  swing2LimitDegrees=45
  swing2Motion=Locked
  twistLimitDegrees=45
  twistMotion=Locked
 }
 {
  bone1Name=base_purseStrap
  bone2Name=spine_03
  swing1LimitDegrees=45
  swing1Motion=Locked
  swing2LimitDegrees=45
  swing2Motion=Locked
  twistLimitDegrees=45
  twistMotion=Locked
 }
 {
  bone1Name=base_purseStrap03
  bone2Name=dyn_purse_02
  swing1LimitDegrees=45
  swing1Motion=Locked
  swing2LimitDegrees=45
  swing2Motion=Locked
  twistLimitDegrees=45
  twistMotion=Locked
 }
 {
  bone1Name=spine_03
  bone2Name=spine_01
  swing1LimitDegrees=45
  swing1Motion=Free
  swing2LimitDegrees=45
  swing2Motion=Free
  twistLimitDegrees=45
  twistMotion=Free
 }
]
