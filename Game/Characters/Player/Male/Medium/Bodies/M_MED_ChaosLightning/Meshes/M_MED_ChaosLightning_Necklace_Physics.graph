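# Physics asset for the necklace on the ChaosLightning medium male player body, defining capsule bodies and constraints for its dangling motion.
# A three-link dyn_necklace chain hangs from base_necklace with tight swing and twist limits, and base_necklace is angularly locked to spine_05.
# Large capsules on the spine and clavicles, also locked together, approximate the torso around the chain.
# A thin base_purseStrap capsule also constrains to spine_05, without locked rotation.
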PhysicsAsset /Game/Characters/Player/Male/Medium/Bodies/M_MED_ChaosLightning/Meshes/M_MED_ChaosLightning_Necklace_Physics
bodies {
 base_necklace [{shapeType=Capsule, length=1, radius=1, shapeName=base_necklace_capsule, rotation=<rot 90 0 0>}]
 base_purseStrap [{
   shapeType=Capsule
   length=9.6
   radius=0.505
   shapeName=base_purseStrap_capsule
   center=<3.6441 2.3361 -0.7951>
   rotation=<rot 1.635 -63.3084 86.7519>
  }]
 clavicle_l [{shapeType=Capsule, length=7.4, radius=6, shapeName=clavicle_l_capsule, center=<4.2225 0.3483 2.5776>, rotation=<rot 90 -87.106 -89.4292>}]
 clavicle_r [{shapeType=Capsule, length=7.4, radius=6, shapeName=clavicle_l_capsule, center=<-4.2225 -0.3483 -2.5776>, rotation=<rot 90 -167.0926 10.5842>}]
 dyn_necklace01 [{shapeType=Capsule, length=6.2, radius=0.6, shapeName=dyn_necklace01_capsule, center=<1.5547 0 0>, rotation=<rot 0 0 90>}]
 dyn_necklace02 [{shapeType=Capsule, length=5, radius=1, shapeName=dyn_necklace02_capsule, center=<2 0 0>, rotation=<rot 0 0 90>}]
 dyn_necklace03 [{shapeType=Capsule, length=3.8, radius=1.7, shapeName=dyn_necklace03_capsule, center=<4.9578 0 0>, rotation=<rot 90 0 0>}]
 spine_03 [{shapeType=Capsule, length=2.6693, radius=13.1928, shapeName=spine_03_capsule, center=<0 0 -0.8079>}]
 spine_05 [{shapeType=Capsule, length=0.2, radius=19.3, shapeName=spine_05_capsule, center=<-3.7076 -8.3908 0>}]
}
constraints [
 {
  bone1Name=base_necklace
  bone2Name=spine_05
  swing1LimitDegrees=45
  swing1Motion=Locked
  swing2LimitDegrees=45
  swing2Motion=Locked
  twistLimitDegrees=45
  twistMotion=Locked
 }
 {bone1Name=dyn_necklace01, bone2Name=base_necklace, swing1LimitDegrees=10, swing2LimitDegrees=20, twistLimitDegrees=10}
 {bone1Name=dyn_necklace02, bone2Name=dyn_necklace01, swing1LimitDegrees=20, swing2LimitDegrees=15, twistLimitDegrees=5}
 {bone1Name=dyn_necklace03, bone2Name=dyn_necklace02, swing1LimitDegrees=20, swing2LimitDegrees=20, twistLimitDegrees=5}
 {
  bone1Name=clavicle_l
  bone2Name=spine_05
  swing1LimitDegrees=45
  swing1Motion=Locked
  swing2LimitDegrees=45
  swing2Motion=Locked
  twistLimitDegrees=45
  twistMotion=Locked
 }
 {
  bone1Name=spine_05
  bone2Name=spine_03
  swing1LimitDegrees=45
  swing1Motion=Locked
  swing2LimitDegrees=45
  swing2Motion=Locked
  twistLimitDegrees=45
  twistMotion=Locked
 }
 {
  bone1Name=clavicle_r
  bone2Name=spine_05
  swing1LimitDegrees=45
  swing1Motion=Locked
  swing2LimitDegrees=45
  swing2Motion=Locked
  twistLimitDegrees=45
  twistMotion=Locked
 }
 {bone1Name=base_purseStrap, bone2Name=spine_05, swing1LimitDegrees=45, swing2LimitDegrees=45, twistLimitDegrees=45}
]
