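# Physics asset for the ChaosLightning medium male player body, covering dangling belt chains and a purse rather than a full ragdoll.
# Belt A hangs from the left thigh and belts B and C from the right, via tiny base capsules locked to the thighs.
# Chain links have narrow swing and minimal twist limits.
# The right calf locks to its thigh, while thighs and purse rotate freely against a large root capsule.
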PhysicsAsset /Game/Characters/Player/Male/Medium/Bodies/M_MED_ChaosLightning/Meshes/M_MED_ChaosLightning_Physics
bodies {
 base_beltA [{shapeType=Capsule, length=0.505, radius=0.505, shapeName=base_beltA_capsule, rotation=<rot 0 -90 0>}]
 base_beltB [{shapeType=Capsule, length=0.505, radius=0.505, shapeName=base_beltB_capsule, rotation=<rot 0 -90 0>}]
 base_beltC [{shapeType=Capsule, length=0.505, radius=0.505, shapeName=base_beltC_capsule, rotation=<rot 0 -90 0>}]
 calf_r [{
   shapeType=Capsule
   length=26.9031
   radius=8.687
   shapeName=calf_r_capsule
   center=<14.3425 2.2099 -0.5527>
   rotation=<rot 0 -86.3948 -87.8099>
  }]
 dyn_beltA_01 [{shapeType=Capsule, length=3.4, radius=1.6, shapeName=dyn_beltA_01_capsule, center=<6.1223 -0.5152 0>, rotation=<rot 0 180 174.628>}]
 dyn_beltA_02 [{shapeType=Capsule, length=2.7, radius=1.6, shapeName=dyn_beltA_02_capsule, center=<1.2796 0 0>}]
 dyn_beltB_01 [{shapeType=Capsule, length=2.3, radius=2.2, shapeName=dyn_beltB_01_capsule, center=<4.9211 -0.2979 0>, rotation=<rot 90 0 0>}]
 dyn_beltB_02 [{shapeType=Capsule, length=1.7, radius=2.1, shapeName=dyn_beltB_02_capsule, center=<3.409 0 0>}]
 dyn_beltB_03 [{shapeType=Capsule, length=1.5, radius=2.8, shapeName=dyn_beltB_03_capsule, center=<1.5695 0 0>}]
 dyn_beltC_01 [{shapeType=Capsule, length=3.8, radius=1.8, shapeName=dyn_beltC_01_capsule, center=<5.8055 -0.463 0>, rotation=<rot 90 0 0>}]
 dyn_beltC_02 [{shapeType=Capsule, length=8.6, radius=2.1, shapeName=dyn_beltC_02_capsule, center=<6.4935 0 0>, rotation=<rot 90 0 0>}]
 dyn_beltC_03 [{shapeType=Capsule, length=8.6, radius=2.1, shapeName=dyn_beltC_03_capsule, center=<6.4935 0 0>, rotation=<rot 90 0 0>}]
 dyn_beltC_04 [{shapeType=Capsule, length=8.6, radius=2.1, shapeName=dyn_beltC_04_capsule, center=<7.9985 0.7992 0>, rotation=<rot 90 180 171.1999>}]
 dyn_purse_03 [{
   shapeType=Capsule
   length=16.1
   radius=5.1
   shapeName=dyn_purse_03_capsule
   center=<17.0955 -0.1931 1.3077>
   rotation=<rot 2.8604 10.8026 66.7069>
  }]
 root [{shapeType=Capsule, length=0, radius=108.7, shapeName=root_capsule, center=<-6.6956 17.7978 -95.1117>}]
 thigh_l [{shapeType=Capsule, length=6.6058, radius=10.064, shapeName=thigh_l_capsule, center=<-20.3722 0.909 0.5359>, rotation=<rot 0 -0.0564 67.5033>}]
 thigh_r [{
   shapeType=Capsule
   length=14.8058
   radius=8.264
   shapeName=thigh_l_capsule
   center=<23.19 0.4307 1.0166>
   rotation=<rot 82.7329 179.9997 -0.0001>
  }]
}
constraints [
 {bone1Name=dyn_beltA_01, bone2Name=base_beltA, swing1LimitDegrees=10, swing2LimitDegrees=6, twistLimitDegrees=1}
 {bone1Name=dyn_beltA_02, bone2Name=dyn_beltA_01, swing1LimitDegrees=35, swing2LimitDegrees=10, twistLimitDegrees=1}
 {bone1Name=dyn_beltB_01, bone2Name=base_beltB, swing1LimitDegrees=15, swing2LimitDegrees=10, twistLimitDegrees=1}
 {bone1Name=dyn_beltB_02, bone2Name=dyn_beltB_01, swing1LimitDegrees=15, swing2LimitDegrees=10, twistLimitDegrees=1}
 {bone1Name=dyn_beltB_03, bone2Name=dyn_beltB_02, swing1LimitDegrees=30, swing2LimitDegrees=10, twistLimitDegrees=1}
 {bone1Name=dyn_beltC_01, bone2Name=base_beltC, swing1LimitDegrees=15, swing2LimitDegrees=30, twistLimitDegrees=1}
 {bone1Name=dyn_beltC_02, bone2Name=dyn_beltC_01, swing1LimitDegrees=10, swing2LimitDegrees=20, twistLimitDegrees=1}
 {bone1Name=dyn_beltC_03, bone2Name=dyn_beltC_02, swing1LimitDegrees=20, swing2LimitDegrees=15, twistLimitDegrees=1}
 {bone1Name=dyn_beltC_04, bone2Name=dyn_beltC_03, swing1LimitDegrees=20, swing2LimitDegrees=10, twistLimitDegrees=1}
 {
  bone1Name=base_beltA
  bone2Name=thigh_l
  swing1LimitDegrees=45
  swing1Motion=Locked
  swing2LimitDegrees=45
  swing2Motion=Locked
  twistLimitDegrees=45
  twistMotion=Locked
 }
 {
  bone1Name=base_beltC
  bone2Name=thigh_r
  swing1LimitDegrees=45
  swing1Motion=Locked
  swing2LimitDegrees=45
  swing2Motion=Locked
  twistLimitDegrees=45
  twistMotion=Locked
 }
 {
  bone1Name=base_beltB
  bone2Name=thigh_r
  swing1LimitDegrees=45
  swing1Motion=Locked
  swing2LimitDegrees=45
  swing2Motion=Locked
  twistLimitDegrees=45
  twistMotion=Locked
 }
 {
  bone1Name=calf_r
  bone2Name=thigh_r
  swing1LimitDegrees=45
  swing1Motion=Locked
  swing2LimitDegrees=45
  swing2Motion=Locked
  twistLimitDegrees=45
  twistMotion=Locked
 }
 {
  bone1Name=dyn_purse_03
  bone2Name=root
  swing1LimitDegrees=45
  swing1Motion=Free
  swing2LimitDegrees=45
  swing2Motion=Free
  twistLimitDegrees=45
  twistMotion=Free
 }
 {
  bone1Name=thigh_l
  bone2Name=root
  swing1LimitDegrees=45
  swing1Motion=Free
  swing2LimitDegrees=45
  swing2Motion=Free
  twistLimitDegrees=45
  twistMotion=Free
 }
 {
  bone1Name=thigh_r
  bone2Name=root
  swing1LimitDegrees=45
  swing1Motion=Free
  swing2LimitDegrees=45
  swing2Motion=Free
  twistLimitDegrees=45
  twistMotion=Free
 }
]
