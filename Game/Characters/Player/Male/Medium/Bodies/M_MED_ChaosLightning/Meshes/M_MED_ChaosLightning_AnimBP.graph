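# Anim Blueprint for the ChaosLightning male player body, adding deformation and secondary motion to the pose copied from its parent mesh.
# A Control Rig feeds bag and strap constraints, corrective offsets and bag RigidBody simulation.
# RigidBody and AnimDynamics then move body, necklace, sleeves, shoes and cords, before shoe and front-end belt offsets.
# Updates derive corrective alphas from spine/clavicle rotations and skydive/parachute state, and compute a gravity override.
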AnimBlueprint /Game/Characters/Player/Male/Medium/Bodies/M_MED_ChaosLightning/Meshes/M_MED_ChaosLightning_AnimBP
target_skeleton=&/M_MED_ChaosLightning_Skeleton
graphs [
 AnimGraph {nodes [
   n1 = RigidBody!(ComponentPose=Local To Component(LocalPose=Use cached pose 'BagEnd'()), Alpha, OverrideWorldGravity:Vector="0,0,-980") -> (Pose) {
    bClampLinearTranslationLimitToRefPose=true
    bOverrideWorldGravity=true
    overridePhysicsAsset=&/M_MED_ChaosLightning_Physics
    simulationSpace=BaseBoneSpace
    alphaBoolBlend {blendOption=Linear}
    baseBoneRef {boneName=pelvis}
    componentAppliedLinearAccClamp=<1000 1000 1000>
    componentLinearAccScale=<0.2 0.1 0.1>
    componentLinearVelScale=<0.2 0.1 0.1>
    overrideWorldGravity {z=-980}
    simSpaceSettings {worldAlpha=0.3}
   }
   n2 = Anim Dynamics - Bone: dyn_sleeve_l(ComponentPose=@n1, Alpha=1, ExternalForce="0,0,0") {
    angularDampingOverride=1
    angularSpringConstant=20000
    bAngularSpring=true
    bOverrideAngularDamping=true
    bUseGravityOverride=true
    gravityScale=0.35
    simulationSpace=RootRelative
    alphaBoolBlend {blendOption=Linear}
    boundBone {boneName=dyn_sleeve_l}
    componentLinearAccScale=<1 1 1>
    componentLinearVelScale=<1 1 1>
    gravityOverride {z=-980}
    physicsBodyDefinitions [{
      boundBone {boneName=dyn_sleeve_l}
      boxExtents=<20 30 30>
      constraintSetup {angularLimitsMax=<0 15 2>, angularLimitsMin=<0 0 -1>, angularTarget=<1 0 0>}
      localJointOffset=<4 0 0>
     }]
   }
   n3 = Anim Dynamics - Bone: base_sleeve_r(ComponentPose=@n2, Alpha=1, ExternalForce="0,0,0") {
    angularDampingOverride=1
    angularSpringConstant=20000
    bAngularSpring=true
    bOverrideAngularDamping=true
    bUseGravityOverride=true
    gravityScale=0.35
    simulationSpace=RootRelative
    alphaBoolBlend {blendOption=Linear}
    boundBone {boneName=base_sleeve_r}
    componentLinearAccScale=<1 1 1>
    componentLinearVelScale=<1 1 1>
    gravityOverride {z=-980}
    physicsBodyDefinitions [{
      boundBone {boneName=base_sleeve_r}
      boxExtents=<20 30 30>
      constraintSetup {angularLimitsMax=<0 15 0>, angularLimitsMin=<0 0 0>, angularTarget=<1 0 0>}
      localJointOffset=<4 0 0>
     }]
   }
   n4 = RigidBody(ComponentPose=@n3, OverrideWorldGravity="0,0,-980") {
    bClampLinearTranslationLimitToRefPose=true
    bOverrideWorldGravity=true
    overridePhysicsAsset=&/M_MED_ChaosLightning_Necklace_Physics
    simulationSpace=BaseBoneSpace
    alphaBoolBlend {blendOption=Linear}
    baseBoneRef {boneName=spine_05}
    componentAppliedLinearAccClamp=<1000 1000 1000>
    componentLinearAccScale=<0.2 0.1 0.1>
    componentLinearVelScale=<0.2 0.1 0.1>
    overrideWorldGravity {z=-980}
    simSpaceSettings {worldAlpha=0.3}
   }
   n5 = Anim Dynamics - Bone: dyn_shoe_r(ComponentPose=@n4, Alpha=1, ExternalForce="0,0,0") {
    angularDampingOverride=1
    angularSpringConstant=20000
    bAngularSpring=true
    bOverrideAngularDamping=true
    bUseGravityOverride=true
    gravityScale=0.35
    simulationSpace=BoneRelative
    alphaBoolBlend {blendOption=Linear}
    boundBone {boneName=dyn_shoe_r}
    componentLinearAccScale=<1 1 1>
    componentLinearVelScale=<1 1 1>
    gravityOverride {z=-980}
    physicsBodyDefinitions [{
      boundBone {boneName=dyn_shoe_r}
      boxExtents=<20 30 30>
      constraintSetup {angularLimitsMax=<1 15 5>, angularLimitsMin=<-1 -15 -15>, angularTarget=<1 0 0>}
      localJointOffset=<4 0 0>
     }]
    relativeSpaceBone {boneName=foot_r}
   }
   n6 = Anim Dynamics - Bone: dyn_shoe_l(ComponentPose=@n5, Alpha=1, ExternalForce="0,0,0") {
    angularDampingOverride=1
    angularSpringConstant=20000
    bAngularSpring=true
    bOverrideAngularDamping=true
    bUseGravityOverride=true
    gravityScale=0.35
    simulationSpace=BoneRelative
    alphaBoolBlend {blendOption=Linear}
    boundBone {boneName=dyn_shoe_l}
    componentLinearAccScale=<1 1 1>
    componentLinearVelScale=<1 1 1>
    gravityOverride {z=-980}
    physicsBodyDefinitions [{
      boundBone {boneName=dyn_shoe_l}
      boxExtents=<20 30 30>
      constraintSetup {angularLimitsMax=<15 15 15>, angularLimitsMin=<-15 -15 -15>, angularTarget=<1 0 0>}
      localJointOffset=<4 0 0>
     }]
    relativeSpaceBone {boneName=foot_l}
   }
   n7 = Anim Dynamics - Bone: dyn_cord_r(ComponentPose=@n6, Alpha=1, ExternalForce="0,0,0") {
    angularDampingOverride=1
    angularSpringConstant=20000
    bAngularSpring=true
    bOverrideAngularDamping=true
    bUseGravityOverride=true
    gravityScale=0.35
    simulationSpace=BoneRelative
    alphaBoolBlend {blendOption=Linear}
    boundBone {boneName=dyn_cord_r}
    componentLinearAccScale=<1 1 1>
    componentLinearVelScale=<1 1 1>
    gravityOverride {z=-980}
    physicsBodyDefinitions [{
      boundBone {boneName=dyn_cord_r}
      boxExtents=<20 30 30>
      constraintSetup {angularLimitsMax=<15 15 15>, angularLimitsMin=<-15 -15 -15>, angularTarget=<1 0 0>}
      localJointOffset=<4 0 0>
     }]
    relativeSpaceBone {boneName=calf_r}
   }
   n8 = Anim Dynamics - Bone: dyn_cord_l(ComponentPose=@n7, Alpha=1, ExternalForce="0,0,0") {
    angularDampingOverride=1
    angularSpringConstant=20000
    bAngularSpring=true
    bOverrideAngularDamping=true
    bUseGravityOverride=true
    gravityScale=0.35
    simulationSpace=BoneRelative
    alphaBoolBlend {blendOption=Linear}
    boundBone {boneName=dyn_cord_l}
    componentLinearAccScale=<1 1 1>
    componentLinearVelScale=<1 1 1>
    gravityOverride {z=-980}
    physicsBodyDefinitions [{
      boundBone {boneName=dyn_cord_l}
      boxExtents=<20 30 30>
      constraintSetup {angularLimitsMax=<15 15 15>, angularLimitsMin=<-15 -15 -15>, angularTarget=<1 0 0>}
      localJointOffset=<4 0 0>
     }]
    relativeSpaceBone {boneName=calf_l}
   }
   n9 = "Transform (Modify) Bone - Bone: base_shoe_l"!(ComponentPose=@n8, bAlphaBoolEnabled:bool=True, Translation="-2.012,0.993,-1.092", Rotation="-12.845,29.941,-0.009") -> (Pose) {
    alphaInputType=Bool
    rotationMode=BMM_Additive
    rotationSpace=BCS_BoneSpace
    translationMode=BMM_Additive
    translationSpace=BCS_BoneSpace
    alphaBoolBlend {blendInTime=0.2, blendOption=Linear, blendOutTime=0.2}
    boneToModify {boneName=base_shoe_l}
    rotation=<rot -12.8449 29.941 -0.0091>
    translation=<-2.012 0.993 -1.092>
   }
   n10 = "Transform (Modify) Bone - Bone: base_shoe_r"!(ComponentPose=@n9, bAlphaBoolEnabled:bool=True, Translation="-1.990,-1.272,-1.392", Rotation="-35.019,-30.026,-30.123") -> (Pose) {
    alphaInputType=Bool
    rotationMode=BMM_Additive
    rotationSpace=BCS_BoneSpace
    translationMode=BMM_Additive
    translationSpace=BCS_BoneSpace
    alphaBoolBlend {blendInTime=0.2, blendOption=Linear, blendOutTime=0.2}
    boneToModify {boneName=base_shoe_r}
    rotation=<rot -35.0191 -30.026 -30.1231>
    translation=<-1.99 -1.272 -1.392>
   }
   n11 = "Transform (Modify) Bone - Bone: dyn_beltA_01"!(ComponentPose=@n10, bAlphaBoolEnabled:bool=Get bIsInFrontEnd(), Translation="0,0,0", Rotation="0.895,7.810,0.111") -> (Pose) {
    alphaInputType=Bool
    rotationMode=BMM_Additive
    rotationSpace=BCS_BoneSpace
    translationMode=BMM_Additive
    translationSpace=BCS_BoneSpace
    alphaBoolBlend {blendOption=Linear}
    boneToModify {boneName=dyn_beltA_01}
    rotation=<rot 0.895 7.81 0.111>
   }
   Output Pose(Result=Component To Local(ComponentPose=@n11)) {layerGroup=None}
   n12 = Control Rig(Source=Copy Pose From Mesh() {bUseAttachedParent=true}) {
    alphaBoolBlend {blendOption=Linear}
    controlRigAssetReference {blueprintRigClass=&/M_MED_DeformRig_ChaosLightning_ControlRig.M_MED_DeformRig_ChaosLightning_ControlRig_C}
   }
   Save cached pose 'StartEnd'(Pose=@n12) {cacheName=StartEnd}
   n13 = Get Clavicle_l_rotY() -> (Clavicle_l_rotY:double)
   n14 = "Constraint - base_bagStrap_front (PPPPPPPP)"!(
    ComponentPose=Local To Component(LocalPose=Use cached pose 'StartEnd'())
    Alpha=1
    ConstraintWeights_0=0.5972
    ConstraintWeights_1:float=0.1325
    ConstraintWeights_2:float=0.0757
    ConstraintWeights_3:float=0.0526
    ConstraintWeights_4:float=0.0387
    ConstraintWeights_5:float=0.0375
    ConstraintWeights_6:float=0.0336
    ConstraintWeights_7:float=0.0322
   ) -> (Pose) {
    alphaBoolBlend {blendOption=Linear}
    boneToModify {boneName=base_bagStrap_front}
    constraintSetup [
     {targetBone {boneName=deform_pec_aim_l}}
     {targetBone {boneName=deform_pec_aim_r}}
     {targetBone {boneName=deform_pec_l}}
     {targetBone {boneName=upperarm_twist_01_l}}
     {targetBone {boneName=spine_05}}
     {targetBone {boneName=neck_01}}
     {targetBone {boneName=deform_upperarm_fr_l}}
     {targetBone {boneName=clavicle_l}}
    ]
    constraintWeights [1,1,1,1,1,1,1,1]
   }
   n15 = "Constraint - base_bagStrap_back (PPPPP)"!(
    ComponentPose=@n14
    Alpha=1
    ConstraintWeights_0=0.3826
    ConstraintWeights_1:float=0.2833
    ConstraintWeights_2:float=0.1777
    ConstraintWeights_3:float=0.1386
    ConstraintWeights_4:float=0.0178
   ) -> (Pose) {
    alphaBoolBlend {blendOption=Linear}
    boneToModify {boneName=base_bagStrap_back}
    constraintSetup [
     {targetBone {boneName=spine_03}}
     {targetBone {boneName=spine_05}}
     {targetBone {boneName=spine_04}}
     {targetBone {boneName=clavicle_l}}
     {targetBone {boneName=spine_02}}
    ]
    constraintWeights [1,1,1,1,1]
   }
   n16 = Get TorsoRot() -> (TorsoRot:double)
   n17 = "Transform (Modify) Bone - Bone: dyn_bagStrap_front"!(ComponentPose=@n15, Alpha=@n16, Translation="2,0,0", Rotation="0,-10,0") -> (Pose) {
    rotationMode=BMM_Additive
    rotationSpace=BCS_BoneSpace
    translationMode=BMM_Additive
    translationSpace=BCS_BoneSpace
    alphaBoolBlend {blendOption=Linear}
    boneToModify {boneName=dyn_bagStrap_front}
    rotation {yaw=-10}
    translation {x=2}
   }
   n18 = "Transform (Modify) Bone - Bone: dyn_bagStrap_back"!(ComponentPose=@n17, Alpha=@n16, Translation="4,0,0", Rotation="0,0,0") -> (Pose) {
    rotationMode=BMM_Additive
    rotationSpace=BCS_BoneSpace
    translationMode=BMM_Additive
    translationSpace=BCS_BoneSpace
    alphaBoolBlend {blendOption=Linear}
    boneToModify {boneName=dyn_bagStrap_back}
    translation {x=4}
   }
   n19 = "Transform (Modify) Bone - Bone: dyn_purseStrap03"!(ComponentPose=@n18, Alpha=1, Rotation="25,0,0") -> (Pose) {
    rotationMode=BMM_Additive
    rotationSpace=BCS_BoneSpace
    alphaBoolBlend {blendOption=Linear}
    boneToModify {boneName=dyn_purseStrap03}
    rotation {pitch=25}
   }
   n20 = "Transform (Modify) Bone - Bone: dyn_purseStrap03"(ComponentPose=@n19, Alpha=Get PurseStrapBack_rotZ_neg() -> (PurseStrapBack_rotZ_neg:double), Rotation="20,0,10") {
    rotationMode=BMM_Additive
    rotationSpace=BCS_BoneSpace
    alphaBoolBlend {blendOption=Linear}
    boneToModify {boneName=dyn_purseStrap03}
    rotation {pitch=20, roll=10}
   }
   n21 = "Transform (Modify) Bone - Bone: dyn_bagStrap_front"!(
    ComponentPose=@n20
    Alpha=Get PurseStrapBack_rotZ_neg()
    Translation="0.535,0.249,0.426"
    Rotation="0.005,-6.037,-0.002"
    Scale:Vector="1,1,1"
   ) -> (Pose) {
    rotationMode=BMM_Additive
    rotationSpace=BCS_BoneSpace
    translationMode=BMM_Additive
    alphaBoolBlend {blendOption=Linear}
    boneToModify {boneName=dyn_bagStrap_front}
    rotation=<rot 0.005 -6.0371 -0.002>
    translation=<0.535 0.249 0.426>
   }
   n22 = Get PurseStrapBack_rotX_neg() -> (PurseStrapBack_rotX_neg:double)
   n23 = "Transform (Modify) Bone - Bone: dyn_bagStrap02"!(ComponentPose=@n21, Alpha=@n22, Translation="-2,1.5,2", Rotation="20,20,-10") -> (Pose) {
    rotationMode=BMM_Additive
    rotationSpace=BCS_BoneSpace
    translationMode=BMM_Additive
    translationSpace=BCS_BoneSpace
    alphaBoolBlend {blendOption=Linear}
    boneToModify {boneName=dyn_bagStrap02}
    rotation=<rot 20 20 -10>
    translation=<-2 1.5 2>
   }
   n24 = "Transform (Modify) Bone - Bone: base_purse_01"!(ComponentPose=@n23, Alpha=1, Translation="0,0,0", Rotation="0,5,0") -> (Pose) {
    rotationMode=BMM_Additive
    rotationSpace=BCS_BoneSpace
    translationMode=BMM_Additive
    translationSpace=BCS_BoneSpace
    alphaBoolBlend {blendOption=Linear}
    boneToModify {boneName=base_purse_01}
    rotation {yaw=5}
   }
   n25 = "Transform (Modify) Bone - Bone: extra_bagStrap_back"!(ComponentPose=@n24, Alpha=@n22, Translation="0,0,3.5", Rotation="0,0,0") -> (Pose) {
    rotationMode=BMM_Additive
    rotationSpace=BCS_BoneSpace
    translationMode=BMM_Additive
    translationSpace=BCS_BoneSpace
    alphaBoolBlend {blendOption=Linear}
    boneToModify {boneName=extra_bagStrap_back}
    translation {z=3.5}
   }
   n26 = "Transform (Modify) Bone - Bone: dyn_bagStrap"!(ComponentPose=@n25, Alpha=@n13, Translation="0.734,-0.133,1.443", Rotation="9.488,-0.662,0") -> (Pose) {
    rotationMode=BMM_Additive
    rotationSpace=BCS_BoneSpace
    translationMode=BMM_Additive
    translationSpace=BCS_BoneSpace
    alphaBoolBlend {blendOption=Linear}
    boneToModify {boneName=dyn_bagStrap}
    rotation {pitch=9.488, yaw=-0.662}
    translation=<0.734 -0.133 1.443>
   }
   n27 = "Transform (Modify) Bone - Bone: dyn_bagStrap_front"!(ComponentPose=@n26, Alpha=@n13, Translation="1,0,0", Rotation="0.005,-10,-10.986") -> (Pose) {
    rotationMode=BMM_Additive
    rotationSpace=BCS_BoneSpace
    translationMode=BMM_Additive
    translationSpace=BCS_BoneSpace
    alphaBoolBlend {blendOption=Linear}
    boneToModify {boneName=dyn_bagStrap_front}
    rotation=<rot 0.005 -10 -10.986>
    translation {x=1}
   }
   n28 = "Transform (Modify) Bone - Bone: dyn_bagStrap02"(ComponentPose=@n27, Alpha=@n13, Translation="2,-0.5,-0.5", Rotation="-6.999,-0.008,21.591") {
    rotationMode=BMM_Additive
    rotationSpace=BCS_BoneSpace
    translationMode=BMM_Additive
    translationSpace=BCS_BoneSpace
    alphaBoolBlend {blendOption=Linear}
    boneToModify {boneName=dyn_bagStrap02}
    rotation=<rot -6.999 -0.008 21.591>
    translation=<2 -0.5 -0.5>
   }
   n29 = "Transform (Modify) Bone - Bone: deform_upperarm_up_base_l"!(ComponentPose=@n28, Alpha=@n13, Translation="6,0,0", Rotation="-29.996,-0.394,1.071") -> (Pose) {
    rotationMode=BMM_Additive
    rotationSpace=BCS_BoneSpace
    translationMode=BMM_Additive
    translationSpace=BCS_BoneSpace
    alphaBoolBlend {blendOption=Linear}
    boneToModify {boneName=deform_upperarm_up_base_l}
    rotation=<rot -29.996 -0.394 1.071>
    translation {x=6}
   }
   n30 = "Transform (Modify) Bone - Bone: base_purseStrap"!(ComponentPose=@n29, Alpha=@n13, Translation="0,0,0", Rotation="0,12,0") -> (Pose) {
    rotationMode=BMM_Additive
    rotationSpace=BCS_BoneSpace
    translationMode=BMM_Additive
    translationSpace=BCS_BoneSpace
    alphaBoolBlend {blendOption=Linear}
    boneToModify {boneName=base_purseStrap}
    rotation {yaw=12}
   }
   n31 = "Transform (Modify) Bone - Bone: base_purse_01"!(ComponentPose=@n30, bAlphaBoolEnabled:bool=True, Rotation="-6.675,9.986,7.498") -> (Pose) {
    alphaInputType=Bool
    rotationMode=BMM_Additive
    rotationSpace=BCS_BoneSpace
    scaleSpace=BCS_BoneSpace
    translationSpace=BCS_BoneSpace
    alphaBoolBlend {blendInTime=0.2, blendOption=Linear, blendOutTime=0.2}
    boneToModify {boneName=base_purse_01}
    rotation=<rot -6.675 9.986 7.498>
   }
   n32 = RigidBody(ComponentPose=@n31, OverrideWorldGravity="0,0,-980") {
    bClampLinearTranslationLimitToRefPose=true
    bOverrideWorldGravity=true
    overridePhysicsAsset=&/M_MED_ChaosLightning_Bag_Physics
    simulationSpace=BaseBoneSpace
    alphaBoolBlend {blendOption=Linear}
    baseBoneRef {boneName=spine_03}
    componentAppliedLinearAccClamp=<6000 6000 3000>
    componentLinearAccScale=<0.2 0.1 0.1>
    componentLinearVelScale=<0.2 0.1 0.1>
    overrideWorldGravity {z=-980}
    simSpaceSettings {worldAlpha=0.25}
   }
   n33 = "Constraint - base_purseStrap03 (PPPP)"!(
    ComponentPose=@n32
    Alpha=1
    ConstraintWeights_0=0.7619
    ConstraintWeights_1:float=0.2207
    ConstraintWeights_2:float=0.0159
    ConstraintWeights_3:float=0.0015
   ) -> (Pose) {
    alphaBoolBlend {blendOption=Linear}
    boneToModify {boneName=base_purseStrap03}
    constraintSetup [
     {targetBone {boneName=dyn_purse_03}}
     {targetBone {boneName=dyn_purse_02}}
     {targetBone {boneName=dyn_purse_01}}
     {targetBone {boneName=dyn_bagStrap_front}}
    ]
    constraintWeights [1,1,1,1]
   }
   n34 = "Transform (Modify) Bone - Bone: dyn_bagStrap_back"!(ComponentPose=@n33, bAlphaBoolEnabled:bool=True, Translation="-1.968,0.376,2.126", Rotation="-0.277,-4.479,-3.549") -> (Pose) {
    alphaInputType=Bool
    rotationMode=BMM_Additive
    rotationSpace=BCS_BoneSpace
    translationMode=BMM_Additive
    alphaBoolBlend {blendInTime=0.2, blendOption=Linear, blendOutTime=0.2}
    boneToModify {boneName=dyn_bagStrap_back}
    rotation=<rot -0.277 -4.479 -3.549>
    translation=<-1.968 0.376 2.126>
   }
   n35 = "Transform (Modify) Bone - Bone: dyn_purseStrap02"!(ComponentPose=@n34, bAlphaBoolEnabled:bool=True, Translation="0,0,0", Rotation="-5,0,0") -> (Pose) {
    alphaInputType=Bool
    rotationMode=BMM_Additive
    rotationSpace=BCS_BoneSpace
    translationMode=BMM_Additive
    alphaBoolBlend {blendInTime=0.2, blendOption=Linear, blendOutTime=0.2}
    boneToModify {boneName=dyn_purseStrap02}
    rotation {pitch=-5}
   }
   n36 = "Transform (Modify) Bone - Bone: dyn_bagStrap"!(ComponentPose=@n35, bAlphaBoolEnabled:bool=True, Translation="0.5,0,0.5", Rotation="15,0,-4") -> (Pose) {
    alphaInputType=Bool
    rotationMode=BMM_Additive
    rotationSpace=BCS_BoneSpace
    translationMode=BMM_Additive
    translationSpace=BCS_BoneSpace
    alphaBoolBlend {blendInTime=0.2, blendOption=Linear, blendOutTime=0.2}
    boneToModify {boneName=dyn_bagStrap}
    rotation {pitch=15, roll=-4}
    translation {x=0.5, z=0.5}
   }
   n37 = "Transform (Modify) Bone - Bone: base_purse_01"!(ComponentPose=@n36, bAlphaBoolEnabled=True, Translation:Vector="-0.000,0.000,-3.373", Rotation="-5,5,-15") -> (Pose) {
    alphaInputType=Bool
    rotationMode=BMM_Additive
    translationMode=BMM_Additive
    alphaBoolBlend {blendInTime=0.2, blendOption=Linear, blendOutTime=0.2}
    boneToModify {boneName=base_purse_01}
    rotation=<rot -5 5 -15>
    translation {z=-3.373}
   }
   n38 = "Transform (Modify) Bone - Bone: dyn_purseStrap03"!(ComponentPose=@n37, bAlphaBoolEnabled:bool=True, Translation:Vector="0,0,0", Rotation="12.426,-10.814,-136.788") -> (Pose) {
    alphaInputType=Bool
    rotationMode=BMM_Additive
    translationMode=BMM_Additive
    translationSpace=BCS_BoneSpace
    alphaBoolBlend {blendInTime=0.2, blendOption=Linear, blendOutTime=0.2}
    boneToModify {boneName=dyn_purseStrap03}
    rotation=<rot 12.426 -10.814 -136.788>
   }
   n39 = "Transform (Modify) Bone - Bone: base_purse_01"!(
    ComponentPose=@n38
    Alpha:float=Get BagFix() -> (BagFix:double)
    Translation="2.742,-2.921,-2.082"
    Rotation="-2.634,-6.560,23.136"
    Scale:Vector="1.000,1.100,1.000"
   ) -> (Pose) {
    rotationMode=BMM_Additive
    scaleMode=BMM_Additive
    scaleSpace=BCS_BoneSpace
    translationMode=BMM_Additive
    translationSpace=BCS_BoneSpace
    alphaBoolBlend {blendOption=Linear}
    boneToModify {boneName=base_purse_01}
    rotation=<rot -2.634 -6.56 23.136>
    scale {y=1.1}
    translation=<2.742 -2.921 -2.082>
   }
   n40 = "Transform (Modify) Bone - Bone: extra_purse_01"!(ComponentPose=@n39, Alpha=Get BagPurseFix() -> (BagPurseFix:double), Translation="1,-6,-2", Rotation="0,-15,0") -> (Pose) {
    rotationMode=BMM_Additive
    translationMode=BMM_Additive
    translationSpace=BCS_BoneSpace
    alphaBoolBlend {blendOption=Linear}
    boneToModify {boneName=extra_purse_01}
    rotation {yaw=-15}
    translation=<1 -6 -2>
   }
   Save cached pose 'BagEnd'(Pose=Component To Local(ComponentPose=@n40)) {cacheName=BagEnd}
  ]}
 EventGraph {nodes [
   Event BlueprintUpdateAnimation()
   > n1 = Sequence!(execute) -> (then_0, then_1, then_2:Exec)
   n2 = Get SkelMesh()
   Set PurseStrapBack_rotZ_neg(
    execute=@n1.then_1
    PurseStrapBack_rotZ_neg:double=GetAlphaFromReferenceTransformDelta(SkeletalMeshComponent=@n2, BoneName=spine_02, BaseName=spine_05, InputRange=(X=0,Y=-15), TwistAxis="1,0,0")
   ) -> (Output_Get:double)
   > Set PurseStrapBack_rotZ_pos(PurseStrapBack_rotZ_pos:double=GetAlphaFromReferenceTransformDelta(SkeletalMeshComponent=@n2, BoneName=spine_02, BaseName=spine_05, InputRange=(X=0,Y=20), TwistAxis="1,0,0")) -> (Output_Get:double)
   > Set PurseStrapBack_rotZ_negFix(PurseStrapBack_rotZ_negFix:double=GetAlphaFromReferenceTransformDelta(SkeletalMeshComponent=@n2, BoneName=spine_04, BaseName=spine_05, InputRange=(X=-2,Y=-6), TwistAxis="1,0,0")) -> (Output_Get:double)
   > Set PurseStrapBack_rotX_neg(PurseStrapBack_rotX_neg:double=GetAlphaFromReferenceTransformDelta(SkeletalMeshComponent=@n2, BoneName=spine_02, BaseName=spine_05, DeltaComponentType=EulerX, InputRange=(X=0,Y=-30), TwistAxis="1,0,0")) -> (Output_Get:double)
   > Set PurseStrapBack_rotX_pos(PurseStrapBack_rotX_pos:double=GetAlphaFromReferenceTransformDelta(SkeletalMeshComponent=@n2, BoneName=spine_02, BaseName=spine_05, DeltaComponentType=EulerX, InputRange=(X=0,Y=20), TwistAxis="1,0,0")) -> (Output_Get:double)
   > Set Clavicle_l_rotY(Clavicle_l_rotY:double=GetAlphaFromReferenceTransformDelta(SkeletalMeshComponent=@n2, BoneName=spine_05, BaseName=clavicle_l, DeltaComponentType=EulerY, InputRange=(X=0,Y=-20), TwistAxis="1,0,0")) -> (Output_Get:double)
   > Set Clavicle_l_rotZ(Clavicle_l_rotZ:double=GetAlphaFromReferenceTransformDelta(SkeletalMeshComponent=@n2, BoneName=spine_05, BaseName=clavicle_l, InputRange=(X=0,Y=-10), TwistAxis="1,0,0")) -> (Output_Get:double)
   n3 = SelectFloat(
    A=GetAlphaFromReferenceTransformDelta(SkeletalMeshComponent=@n2, BoneName=spine_03, BaseName=spine_04, InputRange=(X=0,Y=8), TwistAxis="1,0,0")
    B=0
    bPickA=Get bIsSkyDiving()
   )
   > Set BagFix(BagFix:double=@n3) -> (Output_Get:double)
   > Set BagPurseFix(BagPurseFix:double=SelectFloat(A=1, B=@n3, bPickA=Get bIsParachuteOpened())) -> (Output_Get:double)
   n4 = Get PurseStrapBack_rotZ_neg()
   n5 = Get PurseStrapBack_rotZ_pos() -> (PurseStrapBack_rotZ_pos:double)
   n6 = Get PurseStrapBack_rotX_pos() -> (PurseStrapBack_rotX_pos:double)
   n7 = Get PurseStrapBack_rotX_neg()
   > Set TorsoRot(TorsoRot:double=Remap Float(Input=float + float!(A=@n6, B=@n7, C:double=@n4, D:double=@n5) -> (ReturnValue), OldRange=(X=0,Y=3), NewRange=(X=0,Y=1)) -> (Output=~)) -> (Output_Get:double)
   n8 = "Gravity Override State Function (Parameter Asset) (Nativized)"(
    execute=@n1.then_0
    CharacterPartAnimInstance=Self-Reference()
    GameTimeInSeconds=GetGameTimeInSeconds()
    Parameters=&/M_MED_ChaosLightning_Parameters
   )
   > Set GravityOverride(GravityOverride=@n8)
   Event OnCharacterCustomizationLoaded()
   > Set SkelMesh(SkelMesh=GetPartSkeletalMeshComponent())
  ]}
]
variables [
 BagFix {type=number}
 BagPurseFix {type=number}
 Clavicle_l_rotY {type=number}
 Clavicle_l_rotZ {type=number}
 Gravity Override Params {type=GravityOverrideParamsStruct}
 GravityOverride {type=Vector}
 isCrouchingSprint {type=number}
 isDBNO {type=number}
 isKneeSliding {type=number}
 isParachuteOpen {type=number}
 isSkydive {type=number}
 PurseStrapBack_rotX_neg {type=number}
 PurseStrapBack_rotX_pos {type=number}
 PurseStrapBack_rotZ_neg {type=number}
 PurseStrapBack_rotZ_negFix {type=number}
 PurseStrapBack_rotZ_pos {type=number}
 SkelMesh {type=SkeletalMeshComponent}
 TorsoRot {type=number}
]
